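# Editor utility that generates and saves an "Analytic"-style static mesh asset for every Atom common part type except None.
# Run scales each part description, then CreateGeometry stretches a shared eight-vertex template by radius and height.
# Triangle winding depends on height's sign, and vertex color alpha encodes the part type as a byte fraction.
# Each new asset gets part description and geometry data attached before saving under BasePath.
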EditorUtilityBlueprint /AtomFramework/CommonParts/CreateAnalyticMeshes
graphs [
 ApplyScale {nodes [
   n1 = ApplyScale!() -> (then, Description:Atom Common Part Description Structure, Scale:double=~) {is=title K2Node_FunctionEntry}
   n2 = Break Atom Common Part Description!(AtomCommonPartDescription:"Atom Common Part Description Structure (by ref)"=@n1) -> (Radius:float=~, Height:float=~, InnerRadius:float=~, PlaneQuadrant:Byte, bShowLogo:bool=~) {is=title K2Node_BreakStruct}
   > Return Node!(
    execute
    OutDescription:Atom Common Part Description Structure=Make Atom Common Part Description!(
     Radius:float=float * float(A=@n2, B=@n1.Scale) {is=title K2Node_PromotableOperator}
     Height:float=float * float(A=@n2.Height, B=@n1.Scale)
     InnerRadius:float=float * float(A=@n2.InnerRadius, B=@n1.Scale)
     PlaneQuadrant:Byte=@n2.PlaneQuadrant
     bShowLogo:bool=@n2.bShowLogo
    ) -> (AtomCommonPartDescription:Atom Common Part Description Structure) {is=title K2Node_MakeStruct}
   )
  ]}
 CreateGeometry {nodes [
   n1 = CreateGeometry!() -> (then, Width:double=~, Height:double=~, CommonPartType:EAtomCommonPartType Enum=None) {is=title K2Node_FunctionEntry}
   > n2 = CreateDynamicMeshPool!(execute) -> (then, ReturnValue:Dynamic Mesh Pool Object Reference)
   > n3 = RequestMesh!(execute, self:Dynamic Mesh Pool Object Reference=@n2) -> (then, ReturnValue:Dynamic Mesh Object Reference)
   > Clear(TargetArray:Array of Vectors=Get ScaledVertexPositions() -> (ScaledVertexPositions:Array of Vectors))
   > n4 = For Each Loop(Array:Array of Vectors=Get VertexPositions() -> (VertexPositions:Array of Vectors)) -> (Array Element:Vector)
   > Add(
    TargetArray:Array of Vectors=Get ScaledVertexPositions()
    NewItem:"Vector (by ref)"=vector * vector(A=@n4, B:Vector=MakeVector(X=@n1, Y=@n1, Z=@n1.Height)) -> (ReturnValue=~)
   ) {is=title K2Node_CallArrayFunction}
   n5 = "float >= float"(A=@n1.Height)
   n6 = Get IndexBuffer() -> (IndexBuffer:Array of Int Vector Structures)
   n7 = Get IndexBufferFlipped() -> (IndexBufferFlipped:Array of Int Vector Structures)
   n8 = "To Integer (Byte)"!(InByte:Byte=@n1.CommonPartType) -> (ReturnValue:int=0)
   n9 = MakeColor!(R:float=0, G:float=0, B:float=0, A:float=float / float(A="To Float (Integer)"(InInt=@n8), B=255)) -> (ReturnValue:Linear Color Structure)
   n10 = AppendBuffersToMesh!(
    execute=@n4.Completed
    TargetMesh:Dynamic Mesh Object Reference=@n3
    Buffers:"Geometry Script Simple Mesh Buffers Structure (by ref)"=Make Geometry Script Simple Mesh Buffers!(
     Vertices:Array of Vectors=Get ScaledVertexPositions()
     Normals:Array of Vectors
     UV0:Array of Vector 2D Structures=Get VertexUVs() -> (VertexUVs:Array of Vector 2D Structures)
     UV1:Array of Vector 2D Structures
     UV2:Array of Vector 2D Structures
     UV3:Array of Vector 2D Structures
     UV4:Array of Vector 2D Structures
     UV5:Array of Vector 2D Structures
     UV6:Array of Vector 2D Structures
     UV7:Array of Vector 2D Structures
     VertexColors:Array of Linear Color Structures=Make Array!(
      "[0]":Linear Color Structure=@n9
      "[1]":Linear Color Structure=@n9
      "[2]":Linear Color Structure=@n9
      "[3]":Linear Color Structure=@n9
      "[4]":Linear Color Structure=@n9
      "[5]":Linear Color Structure=@n9
      "[6]":Linear Color Structure=@n9
      "[7]":Linear Color Structure=@n9
     ) -> (Array:Array of Linear Color Structures)
     Triangles:Array of Int Vector Structures=Select(Option 0:Array of Int Vector Structures=@n7, Option 1:Array of Int Vector Structures=@n6, Index=@n5) -> (ReturnValue:Array of Int Vector Structures)
     TriGroupIDs:Array of Integers
    ) -> (GeometryScriptSimpleMeshBuffers:Geometry Script Simple Mesh Buffers Structure) {is=title K2Node_MakeStruct}
    MaterialID:int=0
    bDeferChangeNotifications:bool=false
   ) -> (then, NewTriangleIndicesList:Index List Structure, ReturnValue:Dynamic Mesh Object Reference)
   > n11 = ComputeSplitNormals!(
    execute
    TargetMesh:Dynamic Mesh Object Reference=@n10.ReturnValue
    SplitOptions:Geometry Script Split Normals Options Structure
    CalculateOptions:Geometry Script Calculate Normals Options Structure
   ) -> (then, ReturnValue:Dynamic Mesh Object Reference)
   > Return Node!(execute, DynamicMesh:Dynamic Mesh Object Reference=@n11)
  ]}
 EventGraph {nodes [
   Event Run()
   > n1 = ForEach EAtomCommonPartType!(execute:Exec, SkipHidden:bool) -> (LoopBody:Exec, EnumValue:EAtomCommonPartType Enum, then:Exec) {is=title K2Node_ForEachElementInEnum}
   > n2 = Branch(Condition="Equal (Enum)"(A:EAtomCommonPartType Enum=@n1, B:EAtomCommonPartType Enum=None))
   n3 = Break Atom Common Part Description(AtomCommonPartDescription=ApplyScale!(
     Description:Atom Common Part Description Structure=GetCommonPartDescriptionFromType!(CommonPartType:EAtomCommonPartType Enum=@n1) -> (OutDescription:Atom Common Part Description Structure)
     Scale:double=Get Scale() -> (Scale:double)
    ) -> (OutDescription:Atom Common Part Description Structure) {is=title K2Node_CallFunction}) -> (Radius=~, Height=~, InnerRadius=~, bShowLogo=~)
   n4 = CreateGeometry!(execute=@n2.else, self, Width:double=@n3, Height:double=@n3.Height, CommonPartType:EAtomCommonPartType Enum=@n1) -> (then, DynamicMesh:Dynamic Mesh Object Reference) {is=title K2Node_CallFunction}
   n5 = MakeStandardFilename!(InPath:String=Append(A=Get BasePath() -> (BasePath:String), B=Enum to String!(Enumerator:EAtomCommonPartType Enum=@n1) -> (ReturnValue:String))) -> (OutPath:String)
   > n6 = CreateNewStaticMeshAssetFromMesh!(
    execute
    FromDynamicMesh:Dynamic Mesh Object Reference=@n4
    AssetPathAndName:String=@n5
    Options:Geometry Script Create New Static Mesh Asset Options Structure
   ) -> (Failure:Exec, Success:Exec, ReturnValue:Static Mesh Object Reference)
   AddCommonPartDescriptionToStaticMesh!(
    execute=@n6.Success
    StaticMesh:Static Mesh Object Reference=@n6
    Description:"Atom Common Part Asset Description Structure (by ref)"=Make Atom Common Part Asset Description!(CommonPartType:EAtomCommonPartType Enum=@n1, CommonPartStyle:Name=Analytic, Scale:float=Get Scale()) -> (AtomCommonPartAssetDescription:Atom Common Part Asset Description Structure) {is=title K2Node_MakeStruct}
   ) -> (then)
   > AddCommonPartGeometryToStaticMesh!(execute, StaticMesh:Static Mesh Object Reference=@n6) -> (then)
   > SaveAsset!(execute, AssetToSave:String=@n5, bOnlyIfIsDirty:bool=true) -> (then, ReturnValue:bool)
  ]}
]
variables [
 BasePath {type=string, default=&/Analytic/}
 Scale {type=number, default=1}
]
